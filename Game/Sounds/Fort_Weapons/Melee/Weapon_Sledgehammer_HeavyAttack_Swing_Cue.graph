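# Sound cue for the sledgehammer heavy-attack swing, layering a random generic whoosh with a louder random RatRod hammer impact sample.
# A quality-level node selects between two such mixes, each with randomized upward pitch; its other two slots are empty.
# The first delays the impact slightly, while the second uses fewer variations and a default delay.
# Playback uses small melee attenuation with overrides and player melee-weapon concurrency.
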SoundCue /Game/Sounds/Fort_Weapons/Melee/Weapon_Sledgehammer_HeavyAttack_Swing_Cue
root=QualityLevel_0
nodes [
 {id=Delay_0, children [Random_1]}
 {id=Delay_1, children [Random_0], properties {delayMax=0.05, delayMin=0.05}}
 {id=Mixer_1, children [Random_2,Delay_1], properties {inputVolume [0.8,1.2]}}
 {id=Mixer_2, children [Random_4,Delay_0], properties {inputVolume [0.8,1.2]}}
 {id=Modulator_1, children [Mixer_1], properties {pitchMax=1.3, pitchMin=1}}
 {id=Modulator_2, children [Mixer_2], properties {pitchMax=1.3, pitchMin=1}}
 {id=QualityLevel_0, children [Modulator_1,Modulator_2,null,null]}
 {id=Random_0, children [WavePlayer_0,WavePlayer_1,WavePlayer_2,WavePlayer_3], properties {weights [1,1,1,1]}}
 {id=Random_1, children [WavePlayer_4,WavePlayer_5,WavePlayer_6], properties {weights [1,1,1]}}
 {id=Random_2, children [WavePlayer_9, WavePlayer_10, WavePlayer_11, WavePlayer_12], properties {weights [1,1,1,1]}}
 {id=Random_4, children [WavePlayer_18, WavePlayer_19, WavePlayer_21], properties {weights [1,1,1]}}
 {id=WavePlayer_0, properties {soundWaveAssetPtr=/Game/Sounds/Fort_Weapons/RatRod/Hammer/RR_Hammer_Impact_01}}
 {id=WavePlayer_1, properties {soundWaveAssetPtr=/Game/Sounds/Fort_Weapons/RatRod/Hammer/RR_Hammer_Impact_03}}
 {id=WavePlayer_10, properties {soundWaveAssetPtr=&/rightmouse_generic_whoosh_03}}
 {id=WavePlayer_11, properties {soundWaveAssetPtr=&/rightmouse_generic_whoosh_02}}
 {id=WavePlayer_12, properties {soundWaveAssetPtr=&/rightmouse_generic_whoosh_04}}
 {id=WavePlayer_18, properties {soundWaveAssetPtr=&/rightmouse_generic_whoosh_01}}
 {id=WavePlayer_19, properties {soundWaveAssetPtr=&rightmouse_generic_whoosh_03}}
 {id=WavePlayer_2, properties {soundWaveAssetPtr=/Game/Sounds/Fort_Weapons/RatRod/Hammer/RR_Hammer_Impact_02}}
 {id=WavePlayer_21, properties {soundWaveAssetPtr=&rightmouse_generic_whoosh_04}}
 {id=WavePlayer_3, properties {soundWaveAssetPtr=/Game/Sounds/Fort_Weapons/RatRod/Hammer/RR_Hammer_Impact_04}}
 {id=WavePlayer_4, properties {soundWaveAssetPtr=&RR_Hammer_Impact_01}}
 {id=WavePlayer_5, properties {soundWaveAssetPtr=&RR_Hammer_Impact_03}}
 {id=WavePlayer_6, properties {soundWaveAssetPtr=&RR_Hammer_Impact_02}}
 {id=WavePlayer_9, properties {soundWaveAssetPtr=&rightmouse_generic_whoosh_01}}
]
properties {
 attenuationSettings=&Fort_Weapon_Melee_Sml_Att
 maxDistance=2800
 soundClassObject=&Fort_WeaponMelee_PSM
 volumeMultiplier=1
 attenuationOverrides {distanceAlgorithm=NaturalSound, falloffDistance=2000, attenuationShapeExtents {x=500}}
 concurrencySet [&Player_Melee_Weapon_Conc]
}
